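# Sound cue for the fourth GoggleBud NPC laugh in the LimeRock pre-event audio set.
# A single wave player feeds the output directly, with a slight volume boost and no randomization.
# It uses the pre-event UFO attenuation settings and the Fort_Events sound class, and sends lightly to the distant reverb submix.
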SoundCue /LimeRock/Assets/Audio/SFX/NPCs/IP_SC_LimeRock_PreEvent_GoggleBud_Laugh4
root=WavePlayer_1
nodes [{id=WavePlayer_1, properties {soundWaveAssetPtr=/LimeRock/Assets/Audio/SFX/NPCs/Wav/IP_LimeRock_PreEvent_GoggleBud_Laugh4}}]
properties {
 attenuationSettings=/LimeRock/Assets/Audio/Settings/ATT_LimeRock_PreEvent_UFO
 maxDistance=12033.3789
 soundClassObject=&Fort_Events
 volumeMultiplier=1.3
 soundSubmixSends [{minSendLevel=0, sendLevel=0.1, soundSubmix=&Fortnite_Reverb_Distant}]
}
