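# Sound cue for a Jones NPC foley sound, the "AlrightJones" clip from the LimeRock content.
# A single wave player plays the matching wav directly, with no randomization or modulation nodes.
# It uses the shared LimeRock foley attenuation, routes to the Fort_Events sound class and plays with a boosted volume multiplier.
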SoundCue /LimeRock/Assets/Audio/SFX/NPCs/SC_LimeRock_100_Jones_Foley_03_AlrightJones
root=WavePlayer_0
nodes [{id=WavePlayer_0, properties {soundWaveAssetPtr=/LimeRock/Assets/Audio/SFX/NPCs/Wav/LimeRock_100_Jones_Foley_03_AlrightJones}}]
properties {
 attenuationSettings=/LimeRock/Assets/Audio/Settings/ATT_LimeRock_Foley
 maxDistance=2683.9082
 soundClassObject=&Fort_Events
 volumeMultiplier=1.5
}
