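# Sound cue for the LimeRock NPC Jones gunfire effect, playing a single gunfire wave directly with no randomization or modulation.
# It uses the large POI attenuation with a long audible range, a boosted volume and the Fort_Events sound class.
# A concurrency override caps simultaneous instances per owner with a short retrigger time, and a submix send feeds the distant Fortnite reverb.
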SoundCue /LimeRock/Assets/Audio/SFX/NPCs/SC_LimeRock_100_Jones_GunFire
root=WavePlayer_1
nodes [{id=WavePlayer_1, properties {soundWaveAssetPtr=/LimeRock/Assets/Audio/SFX/NPCs/Wav/LimeRock_100_Jones_GunFire}}]
properties {
 attenuationSettings=/LimeRock/Assets/Audio/Settings/ATT_LimeRock_POI_Large
 maxDistance=40000
 soundClassObject=&Fort_Events
 volumeMultiplier=2.5
 concurrencyOverrides {bLimitToOwner=true, maxCount=2, retriggerTime=0.25}
 concurrencySet [null]
 soundSubmixSends [{minSendLevel=0, sendLevel=0.3, soundSubmix=&Fortnite_Reverb_Distant}]
}
