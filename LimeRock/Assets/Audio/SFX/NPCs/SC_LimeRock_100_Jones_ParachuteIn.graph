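# Sound cue for the Jones NPC parachute-in effect in LimeRock, going by its name and folder.
# A single wave player feeds the root directly, with no randomization or mixing.
# It plays at slightly reduced volume in the events sound class, using a dedicated parachute-in attenuation asset for long-range falloff.
# Concurrency override values specify a small per-owner voice limit and a short retrigger time.
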SoundCue /LimeRock/Assets/Audio/SFX/NPCs/SC_LimeRock_100_Jones_ParachuteIn
root=WavePlayer_0
nodes [{id=WavePlayer_0, properties {soundWaveAssetPtr=/LimeRock/Assets/Audio/SFX/NPCs/Wav/LimeRock_100_Jones_ParachuteIn}}]
properties {
 attenuationSettings=/LimeRock/Assets/Audio/Settings/ATT_LimeRock_ParachuteIn
 maxDistance=23000
 soundClassObject=&Fort_Events
 volumeMultiplier=0.8
 concurrencyOverrides {bLimitToOwner=true, maxCount=2, retriggerTime=0.25}
 concurrencySet [null]
}
