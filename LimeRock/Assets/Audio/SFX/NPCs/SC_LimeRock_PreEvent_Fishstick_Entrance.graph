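# Sound cue for the Fishstick NPC entrance in the LimeRock pre-event.
# A single wave player feeds the entrance sound wave straight to the output, with no randomization or mixing.
# It uses a dedicated Fishstick pre-event attenuation asset, routes to the Fort_Events sound class and plays with a boosted volume multiplier.
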SoundCue /LimeRock/Assets/Audio/SFX/NPCs/SC_LimeRock_PreEvent_Fishstick_Entrance
root=WavePlayer_1
nodes [{id=WavePlayer_1, properties {soundWaveAssetPtr=/LimeRock/Assets/Audio/SFX/NPCs/Wav/LimeRock_PreEvent_Fishstick_Entrance}}]
properties {
 attenuationSettings=/LimeRock/Assets/Audio/Settings/ATT_LimeRock_PreEvent_Fishstick
 maxDistance=6018.5435
 soundClassObject=&Fort_Events
 volumeMultiplier=1.8
}
