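# Sound cue for the Fishstick NPC fall sound effect in the LimeRock pre-event.
# A single wave player feeds the output directly with the fall wav, with no randomization or modulation.
# Playback uses the Fort_Events sound class, a slightly reduced volume and the dedicated pre-event Fishstick attenuation asset for distance falloff.
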SoundCue /LimeRock/Assets/Audio/SFX/NPCs/SC_LimeRock_PreEvent_Fishstick_Fall
root=WavePlayer_0
nodes [{id=WavePlayer_0, properties {soundWaveAssetPtr=/LimeRock/Assets/Audio/SFX/NPCs/Wav/LimeRock_PreEvent_Fishstick_Fall}}]
properties {
 attenuationSettings=/LimeRock/Assets/Audio/Settings/ATT_LimeRock_PreEvent_Fishstick
 maxDistance=6018.5435
 soundClassObject=&Fort_Events
 volumeMultiplier=0.8
}
